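; Market bread prop group Blueprint that randomizes the tint of its food meshes for the Helios commerce environment.
; The construction script calls the parent, then passes random palette colors to SetColoronSecondaryMesh: cheese tints for the summer sausage, bread tints for two white bread meshes.
; The bread loop is nested inside the single-element sausage loop.
; BeginPlay, overlap and tick events are empty.
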(event EventBeginPlay)

(event Collision|EventActorBeginOverlap
  (params (OtherActor :type Actor%20Object%20Reference)))

(event EventTick
  (params (DeltaSeconds :type Float%20%28single-precision%29)))

(fn ConstructionScript
  (|CallParentFunction)
  (for _array_element (Utilities|Array|MakeArray (Variables|Default|GetSM_Commerce_Summer_Sausage_A))
    (Default|SetColoronSecondaryMesh (Utilities|Array|RandomArrayItem (Variables|Default|GetTintColor_Cheese)) _array_element)
    (for _array_element_2 (Utilities|Array|MakeArray (Variables|Default|GetSM_Commerce_WhiteBread_A1) (Variables|Default|GetSM_Commerce_WhiteBread_A))
      (Default|SetColoronSecondaryMesh (Utilities|Array|RandomArrayItem (Variables|Default|GetTintColor_Bread)) _array_element_2))))
