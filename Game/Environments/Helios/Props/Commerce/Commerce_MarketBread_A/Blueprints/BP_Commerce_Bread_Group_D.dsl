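; Blueprint for a market bread prop group of bagels, rolls, focaccia, sourdough, baguettes and white bread, with randomized tints.
; The construction script calls its parent, then loops over seventeen bread mesh components.
; Each passes to SetColoronSecondaryMesh with an unseeded random pick from the TintColor_Bread array, so colors can change on reconstruction.
; BeginPlay, overlap and Tick events are empty.
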(event EventBeginPlay)

(event Collision|EventActorBeginOverlap
  (params (OtherActor :type Actor%20Object%20Reference)))

(event EventTick
  (params (DeltaSeconds :type Float%20%28single-precision%29)))

(fn ConstructionScript
  (|CallParentFunction)
  (for _array_element (Utilities|Array|MakeArray (Variables|Default|GetSM_Commerce_Bagel_A2) (Variables|Default|GetSM_Commerce_Roll_A2) (Variables|Default|GetSM_Commerce_Roll_A3) (Variables|Default|GetSM_Commerce_Focaccia_A2) (Variables|Default|GetSM_Commerce_Focaccia_A3) (Variables|Default|GetSM_Commerce_Sourdough_A) (Variables|Default|GetSM_Commerce_Baguette_A) (Variables|Default|GetSM_Commerce_Baguette_A1) (Variables|Default|GetSM_Commerce_Roll_A1) (Variables|Default|GetSM_Commerce_Bagel_A1) (Variables|Default|GetSM_Commerce_Focaccia_A1) (Variables|Default|GetSM_Commerce_WhiteBread_A) (Variables|Default|GetSM_Commerce_WhiteBread_A1) (Variables|Default|GetSM_Commerce_Bagel_A) (Variables|Default|GetSM_Commerce_Roll_A) (Variables|Default|GetSM_Commerce_WhiteBread_A2) (Variables|Default|GetSM_Commerce_Focaccia_A))
    (Default|SetColoronSecondaryMesh (Utilities|Array|RandomArrayItem (Variables|Default|GetTintColor_Bread)) _array_element)))
